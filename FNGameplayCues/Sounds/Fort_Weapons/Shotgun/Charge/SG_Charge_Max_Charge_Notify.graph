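# Sound cue for the charge shotgun's max-charge notification, using the third-person ranged weapon fire sound class.
# A mixer layers one notify wave with one of three random sweetener waves, and every quality level shares that mixer.
# A modulator randomizes pitch slightly above normal, then charge-shotgun attenuation spatializes the result.
# A concurrency override caps simultaneous instances per owner, stopping the oldest.
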SoundCue /FNGameplayCues/Sounds/Fort_Weapons/Shotgun/Charge/SG_Charge_Max_Charge_Notify
root=SoundClass_2
nodes [
 {
  id=Attenuation_2
  children [Modulator_2]
  properties {attenuationSettings=/FNGameplayCues/Sounds/Attenuation/Fort_Charge_Shotgun_Charging_Att}
 }
 {id=Mixer_3, children [Random_2,Random_0], properties {inputVolume [1,1]}}
 {id=Modulator_2, children [QualityLevel_0], properties {pitchMax=1.1, pitchMin=1.2, volumeMax=1, volumeMin=1}}
 {id=QualityLevel_0, children [Mixer_3,Mixer_3,Mixer_3,Mixer_3]}
 {id=Random_0, children [WavePlayer_1,WavePlayer_2,WavePlayer_4], properties {weights [1,1,1]}}
 {id=Random_2, children [WavePlayer_3], properties {weights [1]}}
 {
  id=SoundClass_2
  children [Attenuation_2]
  properties {soundClassOverride=&Fort_WeaponRanged_Fire_3P_PSM}
 }
 {id=WavePlayer_1, properties {soundWaveAssetPtr=&/SG_Charge_Sweetner_01}}
 {id=WavePlayer_2, properties {soundWaveAssetPtr=&/SG_Charge_Sweetner_02}}
 {id=WavePlayer_3, properties {soundWaveAssetPtr=&/SG_Charge_Notify_Max_Charge_02}}
 {id=WavePlayer_4, properties {soundWaveAssetPtr=&/SG_Charge_Sweetner_03}}
]
properties {
 bDebug=true
 bOverrideConcurrency=true
 maxDistance=3000
 soundClassObject=&Fort_WeaponRanged_Fire_3P_PSM
 volumeMultiplier=0.95
 concurrencyOverrides {bLimitToOwner=true, maxCount=3, resolutionRule=StopOldest}
}
