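# Sound cue for the charge shotgun's charge start, layering a random start sound, a random chunk sound and a quieter steam layer.
# An IsEnemy branch picks the 3P charging attenuation when true and the standard charging attenuation when false or unset.
# Both paths repeat the same mix, with slight pitch and boosted volume randomization under the same 3P fire sound class.
# Concurrency is capped per owner, stopping the oldest instance.
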SoundCue /FNGameplayCues/Sounds/Fort_Weapons/Shotgun/Charge/SG_Charge_Start
root=Branch_0
nodes [
 {
  id=Attenuation_0
  children [Modulator_0]
  properties {attenuationSettings=/FNGameplayCues/Sounds/Attenuation/Fort_Charge_Shotgun_Charging_Att}
 }
 {
  id=Attenuation_2
  children [Modulator_2]
  properties {attenuationSettings=/FNGameplayCues/Sounds/Attenuation/Fort_Charge_Shotgun_Charging_3P_Att}
 }
 {id=Branch_0, children [SoundClass_2,SoundClass_0,SoundClass_0], properties {boolParameterName=IsEnemy}}
 {id=Mixer_0, children [Random_0,Random_1,WavePlayer_9], properties {inputVolume [1,1,0.5]}}
 {id=Mixer_1, children [Random_2,Random_4,WavePlayer_0], properties {inputVolume [1,1,0.5]}}
 {id=Modulator_0, children [Mixer_0], properties {pitchMax=1.1, pitchMin=1, volumeMax=1.7, volumeMin=1.5}}
 {id=Modulator_2, children [Mixer_1], properties {pitchMax=1.1, pitchMin=1, volumeMax=1.7, volumeMin=1.5}}
 {id=Random_0, children [WavePlayer_3,WavePlayer_6], properties {weights [1,1]}}
 {id=Random_1, children [WavePlayer_7,WavePlayer_8], properties {weights [1,1]}}
 {id=Random_2, children [WavePlayer_1,WavePlayer_2], properties {weights [1,1]}}
 {id=Random_4, children [WavePlayer_4,WavePlayer_5], properties {weights [1,1]}}
 {
  id=SoundClass_0
  children [Attenuation_0]
  properties {soundClassOverride=&Fort_WeaponRanged_Fire_3P_PSM}
 }
 {
  id=SoundClass_2
  children [Attenuation_2]
  properties {soundClassOverride=&Fort_WeaponRanged_Fire_3P_PSM}
 }
 {id=WavePlayer_0, properties {soundWaveAssetPtr=&/SG_Charge_Startup_Steam_01_v2}}
 {id=WavePlayer_1, properties {soundWaveAssetPtr=&/SG_Charge_Start_01}}
 {id=WavePlayer_2, properties {soundWaveAssetPtr=&/SG_Charge_Start_02}}
 {id=WavePlayer_3, properties {soundWaveAssetPtr=&SG_Charge_Start_01}}
 {id=WavePlayer_4, properties {soundWaveAssetPtr=&/SG_Charge_Startup_Chunk_02}}
 {id=WavePlayer_5, properties {soundWaveAssetPtr=&/SG_Charge_Startup_Chunk_03}}
 {id=WavePlayer_6, properties {soundWaveAssetPtr=&SG_Charge_Start_02}}
 {id=WavePlayer_7, properties {soundWaveAssetPtr=&SG_Charge_Startup_Chunk_02}}
 {id=WavePlayer_8, properties {soundWaveAssetPtr=&SG_Charge_Startup_Chunk_03}}
 {id=WavePlayer_9, properties {soundWaveAssetPtr=&SG_Charge_Startup_Steam_01_v2}}
]
properties {
 bDebug=true
 bOverrideConcurrency=true
 maxDistance=6300
 soundClassObject=&Fort_WeaponRanged_Fire_3P_PSM
 concurrencyOverrides {bLimitToOwner=true, maxCount=3, resolutionRule=StopOldest}
}
