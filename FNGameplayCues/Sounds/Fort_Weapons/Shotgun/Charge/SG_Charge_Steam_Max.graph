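# Sound cue for the charge shotgun's max steam charge startup sound.
# An IsEnemy branch picks a third-person charging attenuation when true and the standard charging attenuation otherwise.
# Each path plays a like-named wave with slight random pitch variation, under the same third-person ranged-weapon fire sound class.
# Concurrency is limited per owner to a few instances, stopping the oldest.
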SoundCue /FNGameplayCues/Sounds/Fort_Weapons/Shotgun/Charge/SG_Charge_Steam_Max
root=Branch_0
nodes [
 {
  id=Attenuation_0
  children [Modulator_0]
  properties {attenuationSettings=/FNGameplayCues/Sounds/Attenuation/Fort_Charge_Shotgun_Charging_Att}
 }
 {
  id=Attenuation_2
  children [Modulator_2]
  properties {attenuationSettings=/FNGameplayCues/Sounds/Attenuation/Fort_Charge_Shotgun_Charging_3P_Att}
 }
 {id=Branch_0, children [SoundClass_2,SoundClass_0,SoundClass_0], properties {boolParameterName=IsEnemy}}
 {id=Modulator_0, children [WavePlayer_0], properties {pitchMax=1.1, pitchMin=1, volumeMax=1, volumeMin=1}}
 {id=Modulator_2, children [WavePlayer_1], properties {pitchMax=1.1, pitchMin=1, volumeMax=1, volumeMin=1}}
 {
  id=SoundClass_0
  children [Attenuation_0]
  properties {soundClassOverride=&Fort_WeaponRanged_Fire_3P_PSM}
 }
 {
  id=SoundClass_2
  children [Attenuation_2]
  properties {soundClassOverride=&Fort_WeaponRanged_Fire_3P_PSM}
 }
 {id=WavePlayer_0, properties {soundWaveAssetPtr=&/SG_Charge_Startup_Steam_Max_01_v2}}
 {id=WavePlayer_1, properties {soundWaveAssetPtr=&SG_Charge_Startup_Steam_Max_01_v2}}
]
properties {
 bDebug=true
 bOverrideConcurrency=true
 maxDistance=6300
 soundClassObject=&Fort_WeaponRanged_Fire_3P_PSM
 volumeMultiplier=0.6
 concurrencyOverrides {bLimitToOwner=true, maxCount=3, resolutionRule=StopOldest}
}
